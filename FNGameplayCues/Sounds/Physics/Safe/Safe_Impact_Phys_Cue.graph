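# Sound cue for physics impacts of the Safe prop, built from borrowed PropaneTank impact waves and using physics impact attenuation.
# An ImpactType switch picks random small, medium or large waves, with the small set also serving as the default.
# Each set's Strength-driven continuous modulator lowers pitch and raises volume as Strength increases, with a wider volume range for small impacts.
# A root modulator adds slight random pitch and volume variation.
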SoundCue /FNGameplayCues/Sounds/Physics/Safe/Safe_Impact_Phys_Cue
root=Modulator_1
nodes [
 {
  id=ModulatorContinuous_0
  children [Random_0]
  properties {
   pitchModulationParams {default=0, maxOutput=0.8, minOutput=1.2, parameterName=Strength}
   volumeModulationParams {default=0, minOutput=0.2, parameterName=Strength}
  }
 }
 {
  id=ModulatorContinuous_1
  children [Random_1]
  properties {
   pitchModulationParams {default=0, maxOutput=0.8, minOutput=1.2, parameterName=Strength}
   volumeModulationParams {default=0, maxOutput=1.2, minOutput=0.8, parameterName=Strength}
  }
 }
 {
  id=ModulatorContinuous_2
  children [Random_2]
  properties {
   pitchModulationParams {default=0, maxOutput=0.8, minOutput=1.2, parameterName=Strength}
   volumeModulationParams {default=0, maxOutput=1.2, minOutput=0.8, parameterName=Strength}
  }
 }
 {id=Modulator_1, children [Switch_0], properties {pitchMax=1.1, pitchMin=0.9, volumeMax=1.1, volumeMin=0.9}}
 {
  id=Random_0
  children [WavePlayer_0,WavePlayer_1,WavePlayer_2,WavePlayer_3,WavePlayer_4,WavePlayer_5,WavePlayer_6]
  properties {weights [1,1,1,1,1,1,1]}
 }
 {
  id=Random_1
  children [WavePlayer_7, WavePlayer_8, WavePlayer_9, WavePlayer_10, WavePlayer_11, WavePlayer_12, WavePlayer_13, WavePlayer_14]
  properties {weights [1,1,1,1,1,1,1,1]}
 }
 {
  id=Random_2
  children [WavePlayer_15, WavePlayer_16, WavePlayer_17, WavePlayer_18, WavePlayer_19, WavePlayer_20, WavePlayer_21, WavePlayer_22]
  properties {weights [1,1,1,1,1,1,1,1]}
 }
 {
  id=Switch_0
  children [ModulatorContinuous_0, ModulatorContinuous_0, ModulatorContinuous_1, ModulatorContinuous_2]
  properties {intParameterName=ImpactType}
 }
 {id=WavePlayer_0, properties {soundWaveAssetPtr=/FNGameplayCues/Sounds/Physics/PropaneTank/PropaneTank_Impact_Small_01}}
 {id=WavePlayer_1, properties {soundWaveAssetPtr=/FNGameplayCues/Sounds/Physics/PropaneTank/PropaneTank_Impact_Small_02}}
 {id=WavePlayer_10, properties {soundWaveAssetPtr=/FNGameplayCues/Sounds/Physics/PropaneTank/PropaneTank_Impact_Medium_03}}
 {id=WavePlayer_11, properties {soundWaveAssetPtr=/FNGameplayCues/Sounds/Physics/PropaneTank/PropaneTank_Impact_Medium_04}}
 {id=WavePlayer_12, properties {soundWaveAssetPtr=/FNGameplayCues/Sounds/Physics/PropaneTank/PropaneTank_Impact_Medium_05}}
 {id=WavePlayer_13, properties {soundWaveAssetPtr=/FNGameplayCues/Sounds/Physics/PropaneTank/PropaneTank_Impact_Medium_06}}
 {id=WavePlayer_14, properties {soundWaveAssetPtr=/FNGameplayCues/Sounds/Physics/PropaneTank/PropaneTank_Impact_Medium_07}}
 {id=WavePlayer_15, properties {soundWaveAssetPtr=/FNGameplayCues/Sounds/Physics/PropaneTank/PropaneTank_Impact_Large_08}}
 {id=WavePlayer_16, properties {soundWaveAssetPtr=/FNGameplayCues/Sounds/Physics/PropaneTank/PropaneTank_Impact_Large_01}}
 {id=WavePlayer_17, properties {soundWaveAssetPtr=/FNGameplayCues/Sounds/Physics/PropaneTank/PropaneTank_Impact_Large_02}}
 {id=WavePlayer_18, properties {soundWaveAssetPtr=/FNGameplayCues/Sounds/Physics/PropaneTank/PropaneTank_Impact_Large_03}}
 {id=WavePlayer_19, properties {soundWaveAssetPtr=/FNGameplayCues/Sounds/Physics/PropaneTank/PropaneTank_Impact_Large_04}}
 {id=WavePlayer_2, properties {soundWaveAssetPtr=/FNGameplayCues/Sounds/Physics/PropaneTank/PropaneTank_Impact_Small_03}}
 {id=WavePlayer_20, properties {soundWaveAssetPtr=/FNGameplayCues/Sounds/Physics/PropaneTank/PropaneTank_Impact_Large_05}}
 {id=WavePlayer_21, properties {soundWaveAssetPtr=/FNGameplayCues/Sounds/Physics/PropaneTank/PropaneTank_Impact_Large_06}}
 {id=WavePlayer_22, properties {soundWaveAssetPtr=/FNGameplayCues/Sounds/Physics/PropaneTank/PropaneTank_Impact_Large_07}}
 {id=WavePlayer_3, properties {soundWaveAssetPtr=/FNGameplayCues/Sounds/Physics/PropaneTank/PropaneTank_Impact_Small_04}}
 {id=WavePlayer_4, properties {soundWaveAssetPtr=/FNGameplayCues/Sounds/Physics/PropaneTank/PropaneTank_Impact_Small_05}}
 {id=WavePlayer_5, properties {soundWaveAssetPtr=/FNGameplayCues/Sounds/Physics/PropaneTank/PropaneTank_Impact_Small_06}}
 {id=WavePlayer_6, properties {soundWaveAssetPtr=/FNGameplayCues/Sounds/Physics/PropaneTank/PropaneTank_Impact_Small_07}}
 {id=WavePlayer_7, properties {soundWaveAssetPtr=/FNGameplayCues/Sounds/Physics/PropaneTank/PropaneTank_Impact_Medium_08}}
 {id=WavePlayer_8, properties {soundWaveAssetPtr=/FNGameplayCues/Sounds/Physics/PropaneTank/PropaneTank_Impact_Medium_01}}
 {id=WavePlayer_9, properties {soundWaveAssetPtr=/FNGameplayCues/Sounds/Physics/PropaneTank/PropaneTank_Impact_Medium_02}}
]
properties {
 attenuationSettings=/Game/Sounds/Physics/Settings/Physics_Impact_Att
 maxDistance=8512
 pitchMultiplier=1.05
 soundClassObject=&Fort_Impacts
 volumeMultiplier=0.7
}
